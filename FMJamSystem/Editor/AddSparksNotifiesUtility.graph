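# Editor utility that batch-adds or removes Sparks instrument prop spawn notifies on selected AnimSequences, with entry functions for bass, drums, guitar and vocals.
# Each add function calls SetUpNewTrack per prop, which recreates a named notify track and initializes a full-length SpawnSparksProp notify state with socket, prop ID and default mesh.
# RemoveJamInstrumentNotifies deletes all those named tracks.
# Processed sequences are saved and logged.
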EditorUtilityBlueprint /FMJamSystem/Editor/AddSparksNotifiesUtility
graphs [
 AddBassNotifies {nodes [
   AddBassNotifies() {is=title K2Node_FunctionEntry}
   > n1 = GetSelectedAssets()
   > n2 = For Each Loop(Array:Array of Object References=@n1) -> (Array Element:Object Reference)
   PrintString(execute=@n2.Completed, InString=Finished adding bass notifies)
   n3 = Cast To AnimSequence!(execute=@n2.LoopBody, Object=@n2) -> (then, CastFailed, AsAnim Sequence:Anim Sequence Object Reference)
   > SetUpNewTrack!(
    execute
    self
    Sequence:Anim Sequence Object Reference=@n3
    TrackName:Name=SpawnBass
    InstrumentType:ESparksInstrumentType Enum=Bass
    LeftHand:bool=false
    AuxMesh:bool=false
    SocketName:Name=hand_attach_r
    PropId:int=0
    DefaultMesh:Skeletal Mesh Object Reference=/SparksCharacterCommon/Instrument/Bass/Generic/Meshes/Generic_Bass
   ) -> (then)
   > SaveLoadedAsset(AssetToSave=@n2, bOnlyIfIsDirty=false)
   > PrintString(InString=Append(A="Added notify to ", B=GetDisplayName(Object=@n2)))
  ]}
 AddDrumNotifies {nodes [
   AddDrumNotifies() {is=title K2Node_FunctionEntry}
   > n1 = GetSelectedAssets()
   > n2 = For Each Loop(Array=@n1)
   PrintString(execute=@n2.Completed, InString=Finished adding drum notifies)
   n3 = Cast To AnimSequence(execute=@n2.LoopBody, Object=@n2)
   > SetUpNewTrack(
    Sequence=@n3
    TrackName=SpawnDrumstickL
    InstrumentType=Drum
    LeftHand=true
    AuxMesh=false
    SocketName=hand_attach_l
    PropId=0
    DefaultMesh=/SparksCharacterCommon/Instrument/Drum/Generic/Meshes/Generic_Drum
   )
   > SetUpNewTrack(
    Sequence=@n3
    TrackName=SpawnDrumstickR
    InstrumentType=Drum
    LeftHand=false
    AuxMesh=false
    SocketName=hand_attach_r
    PropId=1
    DefaultMesh=&Generic_Drum
   )
   > SetUpNewTrack(
    Sequence=@n3
    TrackName=SpawnDrumKit
    InstrumentType=Drum
    LeftHand=false
    AuxMesh=true
    SocketName=root
    PropId=2
    DefaultMesh=/SparksCharacterCommon/Instrument/DrumKit/Generic/Meshes/Generic_DrumKit
   )
   > SaveLoadedAsset(AssetToSave=@n2, bOnlyIfIsDirty=false)
   > PrintString(InString=Append(A="Added notify to ", B=GetDisplayName(Object=@n2)))
  ]}
 AddGuitarNotifies {nodes [
   AddGuitarNotifies() {is=title K2Node_FunctionEntry}
   > n1 = GetSelectedAssets()
   > n2 = For Each Loop(Array=@n1)
   PrintString(execute=@n2.Completed, InString=Finished adding guitar notifies)
   n3 = Cast To AnimSequence(execute=@n2.LoopBody, Object=@n2)
   > SetUpNewTrack(
    Sequence=@n3
    TrackName=SpawnGuitar
    InstrumentType=Guitar
    LeftHand=false
    AuxMesh=false
    SocketName=hand_attach_r
    PropId=0
    DefaultMesh=/SparksCharacterCommon/Instrument/Guitar/Generic/Meshes/Generic_Guitar
   )
   > SaveLoadedAsset(AssetToSave=@n2, bOnlyIfIsDirty=false)
   > PrintString(InString=Append(A="Added notify to ", B=GetDisplayName(Object=@n2)))
  ]}
 AddVoxNotifies {nodes [
   AddVoxNotifies() {is=title K2Node_FunctionEntry}
   > n1 = GetSelectedAssets()
   > n2 = For Each Loop(Array=@n1)
   PrintString(execute=@n2.Completed, InString=Finished adding vox notifies)
   n3 = Cast To AnimSequence(execute=@n2.LoopBody, Object=@n2)
   > SetUpNewTrack(
    Sequence=@n3
    TrackName=SpawnMicStand
    InstrumentType=Vocals
    LeftHand=true
    AuxMesh=false
    SocketName=weapon_l
    PropId=0
    DefaultMesh=/SparksCharacterCommon/Instrument/MicStand/Generic/Meshes/Generic_MicStand
   )
   > SetUpNewTrack(
    Sequence=@n3
    TrackName=SpawnMic
    InstrumentType=Vocals
    LeftHand=false
    AuxMesh=false
    SocketName=weapon_r
    PropId=1
    DefaultMesh=/SparksCharacterCommon/Instrument/Mic/Generic/Meshes/Generic_Mic
   )
   > SaveLoadedAsset(AssetToSave=@n2, bOnlyIfIsDirty=false)
   > PrintString(InString=Append(A="Added notify to ", B=GetDisplayName(Object=@n2)))
  ]}
 EventGraph {links [], nodes []}
 Make Sure Clear Track Exists {nodes [
   n1 = Make Sure Clear Track Exists!() -> (then, Sequence:Anim Sequence Object Reference, TrackName:Name) {is=title K2Node_FunctionEntry}
   > n2 = Branch(Condition=IsValidAnimNotifyTrackName!(AnimationSequenceBase:Anim Sequence Base Object Reference=@n1, NotifyTrackName:Name=@n1.TrackName) -> (ReturnValue:bool))
   > n3 = RemoveAnimationNotifyTrack!(execute, AnimationSequenceBase:Anim Sequence Base Object Reference=@n1, NotifyTrackName:Name=@n1.TrackName) -> (then)
   AddAnimationNotifyTrack!(
    execute=[@n3.then, @n2.else]
    AnimationSequenceBase:Anim Sequence Base Object Reference=@n1
    NotifyTrackName:Name=@n1.TrackName
    TrackColor:Linear Color Structure=(R=1,G=1,B=1,A=1)
   ) -> (then)
  ]}
 RemoveJamInstrumentNotifies {nodes [
   RemoveJamInstrumentNotifies() {is=title K2Node_FunctionEntry}
   > n1 = GetSelectedAssets()
   > n2 = For Each Loop(Array=@n1)
   PrintString(execute=@n2.Completed, InString=Finished removing jam instrument spawn notifies)
   n3 = Cast To AnimSequence(execute=@n2.LoopBody, Object=@n2)
   > n4 = Sequence!(execute) -> (then_0, then_1, then_2:Exec, then_3:Exec, then_4:Exec, then_5:Exec, then_6:Exec, then_7:Exec)
   > RemoveAnimationNotifyTrack(AnimationSequenceBase=@n3, NotifyTrackName=SpawnGuitar)
   RemoveAnimationNotifyTrack(execute=@n4.then_1, AnimationSequenceBase=@n3, NotifyTrackName=SpawnBass)
   RemoveAnimationNotifyTrack(execute=@n4.then_2, AnimationSequenceBase=@n3, NotifyTrackName=SpawnMicStand)
   RemoveAnimationNotifyTrack(execute=@n4.then_3, AnimationSequenceBase=@n3, NotifyTrackName=SpawnMic)
   RemoveAnimationNotifyTrack(execute=@n4.then_4, AnimationSequenceBase=@n3, NotifyTrackName=SpawnDrumstickL)
   RemoveAnimationNotifyTrack(execute=@n4.then_5, AnimationSequenceBase=@n3, NotifyTrackName=SpawnDrumstickR)
   RemoveAnimationNotifyTrack(execute=@n4.then_6, AnimationSequenceBase=@n3, NotifyTrackName=SpawnDrumKit)
   SaveLoadedAsset(execute=@n4.then_7, AssetToSave=@n3, bOnlyIfIsDirty=false)
   > PrintString(InString=Append(A="Removed notifies from ", B=GetDisplayName(Object=@n3)))
  ]}
 SetUpNewTrack {nodes [
   n1 = SetUpNewTrack!() -> (
    then
    Sequence:Anim Sequence Object Reference
    TrackName:Name
    InstrumentType:ESparksInstrumentType Enum=Guitar
    LeftHand:bool=~
    AuxMesh:bool=~
    SocketName:Name
    PropId:int
    DefaultMesh:Skeletal Mesh Object Reference
   ) {is=title K2Node_FunctionEntry}
   > Make Sure Clear Track Exists!(execute, self, Sequence:Anim Sequence Object Reference=@n1, TrackName:Name=@n1.TrackName) -> (then) {is=title K2Node_CallFunction}
   > n2 = AddAnimationNotifyStateEvent!(
    execute
    AnimationSequenceBase:Anim Sequence Base Object Reference=@n1
    NotifyTrackName:Name=@n1.TrackName
    StartTime:float=0
    Duration:float=GetSequenceLength!(AnimationSequenceBase:Anim Sequence Base Object Reference=@n1) -> (Length:float)
    NotifyStateClass:Anim Notify State Class Reference=/Script/SparksCoreCosmeticsRuntime.FortAnimNotifyState_SpawnSparksProp
   ) -> (then, ReturnValue:Anim Notify State Object Reference)
   > n3 = Cast To FortAnimNotifyState_SpawnSparksProp!(execute, Object=@n2) -> (then, CastFailed, AsFort Anim Notify State Spawn Sparks Prop:Fort Anim Notify State Spawn Sparks Prop Object Reference) {is=title K2Node_DynamicCast}
   > InitNewNotify!(
    execute
    self:Fort Anim Notify State Spawn Sparks Prop Object Reference=@n3
    InType:ESparksInstrumentType Enum=@n1.InstrumentType
    InLeftHand:bool=@n1.LeftHand
    InAuxMesh:bool=@n1.AuxMesh
    InSocketName:Name=@n1.SocketName
    InSkipActorsWithTag:Gameplay Tag Query Structure=(TokenStreamVersion=0,TagDictionary=,QueryTokenStream=,UserDescription="",AutoDescription="")
    InPropId:int=@n1.PropId
    InDefaultMesh:Skeletal Mesh Object Reference=@n1.DefaultMesh
   ) -> (then)
   PrintString(execute=@n3.CastFailed, InString=Failed to cast notify)
  ]}
]
variables []
